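# Unlit translucent Niagara sprite master material that draws a procedural, textureless radial glow.
# Two tunable power-curve falloffs from the UV center blend into opacity, scaled by particle alpha.
# Particle color gets a radially scaled hue shift and optional time-of-day tint, computed per vertex, before emissive adjustments.
# Opacity also fades by camera distance, view angle and scene depth (optionally particle-radius based), with a switchable inverse depth fade.
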Material /Game/Effects/Niagara/MasterMaterials/Glow/M_Glow_Translucent
graph {
 properties {
  bUsedWithNiagaraSprites=true
  blendMode=BLEND_Translucent
  pixelDepthOffsetMode=PDOM_Legacy
  previewMesh=/Engine/Transient.MaterialEditorMeshComponent_23
  shadingModel=MSM_Unlit
  usedShadingModels=Unlit
 }
 nodes [
  n1 = TextureCoordinate()
  n2 = Constant() {r=0.5}
  n3 = Saturate(None=Multiply(A=Distance(A=@n1, B=@n2)) {constB=2})
  n4 = ParticleColor()
  n5 = MaterialFunctionCall(Hue Shift Percentage=Multiply(A=@n3, B=ScalarParameter() {parameterName=HueShift}), Texture=@n4.RGB) {materialFunction=&HueShift}
  n6 = VertexInterpolator(VS=MaterialFunctionCall(
    Color_Morning=VectorParameter() {group=TOD, parameterName=Morning, defaultValue=<rgba 1 1 1 1>}.RGB
    Color_Day=VectorParameter() {group=TOD, parameterName=Day, defaultValue=<rgba 1 1 1 1>}.RGB
    Color_Evening=VectorParameter() {group=TOD, parameterName=Evening, defaultValue=<rgba 1 1 1 1>}.RGB
    Color_Night=VectorParameter() {group=TOD, parameterName=Night, defaultValue=<rgba 1 1 1 1>}.RGB
   ) {materialFunction=&Color_Time_of_Day}.Result)
  MaterialFunctionCall(Emissive=LinearInterpolate(A=@n5.Result, B=Multiply(A=@n6.PS, B=@n5.Result), Alpha=ScalarParameter() {group=TOD, parameterName=Use Time of Day})) {materialFunction=&MF_FX_EmissiveAdjustments}
  n7 = ScalarParameter() {parameterName=Cam Fade Length}
  n8 = ScalarParameter() {parameterName=Cam Fade Offset}
  n9 = OneMinus(None=@n3)
  n10 = VectorParameter() {parameterName=Refine, defaultValue=<rgba 1 1 1 0.5>}
  n11 = VectorParameter() {parameterName=Refine2, defaultValue=<rgba 1 1 1 0.5>}
  n12 = Power(Base=OneMinus(None=Power(Base=@n9, Exp=@n10.R)), Exp=@n11.R)
  n13 = Power(Base=OneMinus(None=Power(Base=@n9, Exp=@n10.B)), Exp=@n11.G)
  n14 = LinearInterpolate(A=Multiply(A=OneMinus(None=@n12), B=@n10.G), B=Multiply(A=OneMinus(None=@n13), B=@n10.A))
  n15 = DepthFade(
   Opacity=Multiply(
    A=MaterialFunctionCall(Fade Length=@n7, Fade Offset=@n8) {materialFunction=&CameraDepthFade}.Result
    B=Multiply(A=@n4.A, B=@n14)
   )
   FadeDistance=LinearInterpolate(
    A=ScalarParameter() {defaultValue=10, parameterName=Depth Fade Distance}
    B=ParticleRadius()
    Alpha=ScalarParameter() {parameterName=Use Particle Radius Depth Fade}
   )
  )
  n16 = OneMinus(None=DepthFade(FadeDistance=ScalarParameter() {defaultValue=10, parameterName=DF_1-x}))
  Multiply(
   A=StaticSwitchParameter(True=Multiply(A=@n15, B=@n16), False=@n15) {parameterName="1-x  depthfade"}
   B=MaterialFunctionCall("FadeSensitivity (optional)"=ScalarParameter() {parameterName=FadeBasedOnViewAngle}) {materialFunction=&FadeBasedOnViewAngle}.Result
  )
 ]
}
