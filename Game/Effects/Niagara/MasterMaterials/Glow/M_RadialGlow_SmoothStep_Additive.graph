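# Unlit additive master material for particle sprites and mesh particles that draws a soft radial glow from a SmoothStep-shaped radial gradient tinted by particle color.
# Color optionally blends with a vertex-interpolated time-of-day tint, then passes through the shared FX emissive adjustments.
# Alpha combines shape, particle alpha and an optional view-angle fade.
# Near-camera fade and depth fade, with distance from a dynamic parameter, are skipped on mobile.
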Material /Game/Effects/Niagara/MasterMaterials/Glow/M_RadialGlow_SmoothStep_Additive
graph {
 properties {
  bUsedWithNanite=true
  bUsedWithNiagaraMeshParticles=true
  bUsedWithNiagaraSprites=true
  bUsedWithParticleSprites=true
  blendMode=BLEND_Additive
  pixelDepthOffsetMode=PDOM_Legacy
  shadingModel=MSM_Unlit
  usedShadingModels=Unlit
 }
 nodes [
  n1 = ParticleColor()
  n2 = SmoothStep(
   Min=ScalarParameter() {group=Shape, parameterName=SmoothStepMin}
   Max=ScalarParameter() {defaultValue=1, group=Shape, parameterName=SmoothStepMax}
   Value=MaterialFunctionCall!(Radius=ScalarParameter() {defaultValue=0.5, group=Shape, parameterName=Radius}, CenterPosition) {materialFunction=/Engine/Functions/Engine_MaterialFunctions01/Gradient/RadialGradient}.RadialGradient
  )
  n3 = DepthFade(
   Opacity=MaterialFunctionCall(Distance=ScalarParameter() {defaultValue=100, parameterName=NearCameraFadeDistance}) {materialFunction=&Near_Camera_Fade}.Result
   FadeDistance=DynamicParameter() {defaultValue=<rgba 0 0 0 1>, paramNames [FadeDistance,Param2,Param3,Param4]}.FadeDistance
  )
  n4 = Multiply(A=Multiply(A=@n1.A, B=@n2), B=ShadingPathSwitch(Default=@n3, Deferred=@n3, Mobile=Constant() {r=1}) {desc=Migrated from FeatureLevelSwitch})
  n5 = Multiply(A=DotProduct(A=PixelNormalWS(), B=CameraVectorWS())) {constB=1.5}
  LinearInterpolate(A=@n4, B=Multiply(A=@n4, B=Saturate(None=@n5)), Alpha=ScalarParameter() {parameterName=Use Dot Angle Fade})
  n6 = Multiply(A=@n1.RGB, B=@n2)
  n7 = VertexInterpolator(VS=MaterialFunctionCall(
    Color_Morning=VectorParameter() {group=TOD, parameterName=Morning, defaultValue=<rgba 1 1 1 1>}.RGB
    Color_Day=VectorParameter() {group=TOD, parameterName=Day, defaultValue=<rgba 1 1 1 1>}.RGB
    Color_Evening=VectorParameter() {group=TOD, parameterName=Evening, defaultValue=<rgba 1 1 1 1>}.RGB
    Color_Night=VectorParameter() {group=TOD, parameterName=Night, defaultValue=<rgba 1 1 1 1>}.RGB
   ) {materialFunction=&Color_Time_of_Day}.Result)
  MaterialFunctionCall(Emissive=LinearInterpolate(A=@n6, B=Multiply(A=@n7.PS, B=@n6), Alpha=ScalarParameter() {group=TOD, parameterName=Use Time of Day})) {materialFunction=&MF_FX_EmissiveAdjustments}
 ]
}
