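# Additive Default Lit material for Niagara sprites that draws a soft round glow tinted by particle color.
# The shape is procedural: a UV sphere mask, raised to an exponent and eased through a sine remap, scales particle alpha.
# Color can blend toward a time-of-day tint before an emissive adjustment function.
# Opacity gets optional near-camera fade plus depth fade, which the mobile shading path skips.
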Material /Game/Effects/Niagara/MasterMaterials/Glow/M_SoftGLow_DefaultLit_Additive
graph {
 properties {bUsedWithNiagaraSprites=true, blendMode=BLEND_Additive, pixelDepthOffsetMode=PDOM_Legacy, usedShadingModels=Default Lit}
 nodes [
  n1 = ParticleColor()
  n2 = MaterialFunctionCall(
   Color_Morning=VectorParameter() {group=TOD, parameterName=Morning, defaultValue=<rgba 1 1 1 1>}.RGB
   Color_Day=VectorParameter() {group=TOD, parameterName=Day, defaultValue=<rgba 1 1 1 1>}.RGB
   Color_Evening=VectorParameter() {group=TOD, parameterName=Evening, defaultValue=<rgba 1 1 1 1>}.RGB
   Color_Night=VectorParameter() {group=TOD, parameterName=Night, defaultValue=<rgba 1 1 1 1>}.RGB
  ) {materialFunction=&Color_Time_of_Day}
  MaterialFunctionCall(Emissive=LinearInterpolate(A=@n1.RGB, B=Multiply(A=@n1.RGB, B=@n2.Result), Alpha=ScalarParameter() {group=TOD, parameterName=Use Time of Day})) {materialFunction=&MF_FX_EmissiveAdjustments}
  n3 = Constant() {r=0.5}
  n4 = Power(
   Base=SphereMask(A=TextureCoordinate(), B=@n3, Radius=@n3, Hardness=Constant())
   Exp=ScalarParameter() {defaultValue=2, parameterName=Opacity Exponent}
  )
  n5 = Sine(None=Add(A=Multiply(A=@n4) {constB=0.5}) {constB=-0.25})
  n6 = Multiply(A=@n1.A, B=Saturate(None=ConstantBiasScale(None=@n5)))
  n7 = MaterialFunctionCall(Distance=ScalarParameter() {defaultValue=800, parameterName=NearCameraFadeDistance}) {materialFunction=&Near_Camera_Fade}
  n8 = ScalarParameter() {defaultValue=15, parameterName=DepthFade}
  Multiply(
   A=LinearInterpolate(A=@n6, B=Multiply(A=@n6, B=@n7.Result), Alpha=ScalarParameter() {parameterName=UseNearCameraFade})
   B=ShadingPathSwitch(Default=DepthFade(FadeDistance=@n8), Mobile=Constant() {r=1}) {desc=Migrated from FeatureLevelSwitch}
  )
 ]
}
