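# Two-sided unlit translucent master material that draws a soft radial gradient glow for Niagara sprite and mesh particles.
# A centered UV sphere mask is shaped by blending between a Power curve and SmoothStep, then multiplied by particle alpha and depth, view-angle and two near-camera fades.
# Dynamic parameters drive power, depth-fade multiplier, angle fade and camera-fade distance per particle.
# Particle color runs through a shared emissive-adjustment function.
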Material /Game/Effects/Niagara/MasterMaterials/Glow/M_TwoSided_Radial_Gradient_Master
graph {
 properties {
  bUsedWithNanite=true
  bUsedWithNiagaraMeshParticles=true
  bUsedWithNiagaraSprites=true
  blendMode=BLEND_Translucent
  pixelDepthOffsetMode=PDOM_Legacy
  shadingModel=MSM_Unlit
  twoSided=true
  usedShadingModels=Unlit
 }
 nodes [
  n1 = ParticleColor()
  MaterialFunctionCall(Emissive=@n1.RGB) {materialFunction=&MF_FX_EmissiveAdjustments}
  n2 = SphereMask(
   A=TextureCoordinate()
   B=Constant() {r=0.5}
   Radius=ScalarParameter() {defaultValue=0.5, group=Shape, parameterName=Radius}
   Hardness=ScalarParameter() {group=Shape, parameterName=Hardness}
  ) {hardnessPercent=0}
  n3 = DynamicParameter() {defaultValue=<rgba 3 1 0 1>, paramNames [Power, DepthFadeDistMult, AngleFade, CameraFade]}
  n4 = Clamp(None=@n3.Power) {clampMode=CMODE_ClampMin, maxDefault=100, minDefault=0.5}
  n5 = ScalarParameter() {group=Shape, parameterName=SmoothStepMin}
  n6 = ScalarParameter() {defaultValue=1, group=Shape, parameterName=SmoothStepMax}
  n7 = ScalarParameter() {defaultValue=100, parameterName=NearCameraFadeDist}
  n8 = ScalarParameter() {defaultValue=100, parameterName=DepthFadeDist}
  n9 = Multiply(
   A=DepthFade(
    Opacity=MaterialFunctionCall(Distance=@n7) {materialFunction=&Near_Camera_Fade}.Result
    FadeDistance=Multiply(A=@n3.DepthFadeDistMult, B=@n8)
   )
   B=MaterialFunctionCall("FadeSensitivity (optional)"=@n3.AngleFade) {materialFunction=&FadeBasedOnViewAngle}.Result
  )
  n10 = MaterialFunctionCall(Distance=@n3.CameraFade) {materialFunction=&Near_Camera_Fade}
  Multiply(
   A=LinearInterpolate(
    A=Power(Base=@n2, Exp=@n4) {constExponent=3}
    B=SmoothStep(Min=@n5, Max=@n6, Value=@n2)
    Alpha=ScalarParameter() {group=Shape, parameterName=Use SmoothStep}
   ) {constAlpha=0}
   B=Multiply(A=@n1.A, B=Multiply(A=@n9, B=@n10.Result))
  )
 ]
}
